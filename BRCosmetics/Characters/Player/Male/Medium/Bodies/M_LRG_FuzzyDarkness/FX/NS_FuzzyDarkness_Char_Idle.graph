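# Niagara system for the FuzzyDarkness character's idle effect, emitting goop sprites from skeletal mesh sockets.
# Spawn groups for ears, chest, arms and hands select socket, shape, velocity and lifetime, with owner speed lowering spawn rate and lifetime.
# Sprites inherit movement, drift under curved forces and die near a body kill point.
# User flags and camera distance control visibility; a near-duplicate second emitter is disabled.
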NiagaraSystem /BRCosmetics/Characters/Player/Male/Medium/Bodies/M_LRG_FuzzyDarkness/FX/NS_FuzzyDarkness_Char_Idle
dependencies {
 usedDataInterfaces [&NiagaraDataInterfaceCamera, &NiagaraDataInterfaceCurve, &NiagaraDataInterfaceSkeletalMesh]
 usedDynamicInputs []
 usedModules [
  &self:SystemSpawnScript.SystemScriptSource.SystemScriptGraph.NiagaraNodeAssignment_2.SetVariables_#1
  &SystemState
  &self:SystemSpawnScript.SystemScriptSource.SystemScriptGraph.NiagaraNodeAssignment_0.SetVariables_#2
  &SampleSkeletalMesh
  &self:SystemSpawnScript.SystemScriptSource.SystemScriptGraph.NiagaraNodeAssignment_1.SetVariables_#3
 ]
}
emitters {
 Goop_All {
  emitterUpdateScript [
   &EmitterState()
   /SetVariables(Emitter.MovementSpawnAlpha:NiagaraFloat=&ClampFloat(Float=&Divide_Float(A=&VectorLength(Vector=$Engine.Owner.Velocity), B=500))) {isSetParameters=true, module=SetVariables_#4}
   &SpawnRate(
    Spawn Probability=0.75
    SpawnGroup:NiagaraInt32=&RandomRangeInt(Minimum=1, Maximum=3, Evaluation Type=::NewEnumerator1)
    SpawnRate=&Lerp_Float(A=8, B=3.5, Alpha=$Emitter.MovementSpawnAlpha)
    Use Spawn Probability=true
   )
   &SpawnRate(
    Spawn Probability=0.75
    SpawnRate=&Lerp_Float(A=8, B=4.5, Alpha=$Emitter.MovementSpawnAlpha)
    Use Spawn Probability=true
   ) {module=SpawnRate001}
   /SetVariables(Emitter.ADS:NiagaraBool=&SetBoolByFloatComparison(
     A=&DistanceBetweenPositions(
      A=&ReturnCameraPosition(
       Coordinate Space:ENiagaraCoordinateSpace=::Simulation
       Camera Query:NiagaraDataInterfaceCamera=DataInterface() {propertyValues {bRequireCurrentFrameData=false}}
      )
      B=$Engine.Emitter.SimulationPosition
     )
     B=&FN_FrontEndFloatModifier(Gameplay=250, Front End=0)
     Comparison Type=::NewEnumerator6
    )) {isSetParameters=true, module=SetVariables_#5}
  ]
  particleSpawnScript [
   &PartitionParticles(Number Of Partitions=4)
   /SetVariables(
    Particles.bIsArm:NiagaraBool=&SetBoolByIntComparison(A=$Emitter.SpawnGroup, B=2, Comparison Type=::NewEnumerator4)
    Particles.bIsEar:NiagaraBool=&SetBoolByIntComparison(A=$Emitter.SpawnGroup, B=0, Comparison Type=::NewEnumerator4)
    Particles.bIsHand:NiagaraBool=&SetBoolByIntComparison(A=$Emitter.SpawnGroup, B=3, Comparison Type=::NewEnumerator4)
    Particles.bIsMainChest:NiagaraBool=&SetBoolByIntComparison(A=$Emitter.SpawnGroup, B=1, Comparison Type=::NewEnumerator4)
   ) {isSetParameters=true, module=SetVariables_#6}
   &InitializeParticle(
    Color=<rgba 0.8 0.8 0.8 3>
    Lifetime=&Multiply_Float(
     A=&Lerp_Float(
      A=&RandomRangeFloat(Minimum=1.4, Maximum=2.2)
      B=0.8
      Alpha=&Divide_Float(A=&VectorLength(Vector=$Engine.Owner.Velocity), B=5000)
     )
     B=&MergeHelper_FloatSwitch(0=0.8, 1=0.7, 2=1, 3=1, Index=$Emitter.SpawnGroup, Num=::NewEnumerator2)
    )
    Mesh Scale Mode=::NewEnumerator1
    Mesh Uniform Scale Max=0.4
    Mesh Uniform Scale Min=0.25
    Sprite Size Mode=::NewEnumerator1
    UsePositionOffset=false
    Write Material Random=true
   )
   &InitialMeshOrientation(Use Rotation Vector=false)
   &SkeletalMeshLocation(
    Bone / Socket Index=&MergeHelper_IntSwitch(
     0=&RandomRangeInt(Minimum=0, Maximum=1)
     1=2
     2=&RandomRangeInt(Minimum=3, Maximum=4)
     3=&RandomRangeInt(Minimum=5, Maximum=6)
     4=?
     5=?
     Index=$Emitter.SpawnGroup
     Num=::NewEnumerator2
    )
    MaskSpawnGroup=false
    Mesh Sampling Type=::NewEnumerator5
    Orientation Sampling=::NewEnumerator1
    Skeletal Mesh=$System.SkeletalMesh
    Socket Sampling Mode=::NewEnumerator5
   )
   &ShapeLocation(
    Additional Quaternion Rotation=true
    Additional Yaw / Pitch / Roll=<0 90 0>
    Apply To Particle Position=::NewEnumerator1
    Cylinder Height=20
    Cylinder Radius=3
    MaskSpawnGroup=false
    Override Seed=false
    Rotation Angle Type:ENiagara_AngleInput=::NewEnumerator0
    Rotation Mode=::NewEnumerator3
    Rotation Quaternion=$Particles.SkeletalMeshLocation.SampledRotation
    Shape Primitive=::NewEnumerator1
    Surface Only=false
    UseOwnerScale=true
   )
   &ShapeLocation(
    Additional Quaternion Rotation=true
    Additional Yaw / Pitch / Roll=<0 0 0.25>
    Apply To Particle Position=::NewEnumerator1
    Handle Radius=4
    Large Radius=7
    MaskSpawnGroup=false
    Override Seed=false
    Rotation Mode=::NewEnumerator3
    Rotation Quaternion=$Particles.SkeletalMeshLocation.SampledRotation
    Shape Primitive=::NewEnumerator3
    UseOwnerScale=true
   ) {module=ShapeLocation001}
   &ShapeLocation(
    Apply To Particle Position=::NewEnumerator1
    MaskSpawnGroup=false
    Sphere Distribution=::NewEnumerator1
    Sphere Radius=&MergeHelper_FloatSwitch(0=0, 1=0, 2=11.5, 3=5, 4=0, Index=$Emitter.SpawnGroup, Num=::NewEnumerator3)
    U Position=&RandomRangeFloat(Minimum=0, Maximum=1)
    UseOwnerScale=true
    V Position=&RandomRangeFloat(Minimum=0, Maximum=0.5)
   ) {module=ShapeLocation002}
   &AddVelocity(
    Velocity=&MergeHelper_VectorSwitch(
     0=$Particles.ShapeLocation.ShapeNormal
     1=$Particles.ShapeLocation001.ShapeNormal
     2=$Particles.ShapeLocation002.ShapeNormal
     3=$Particles.ShapeLocation002.ShapeNormal
     Index=$Emitter.SpawnGroup
     Num=::NewEnumerator2
    )
    Velocity Speed Scale=&Multiply_Float(
     A=&MergeHelper_FloatSwitch(0=20, 1=0, 2=40, 3=90, Index=$Emitter.SpawnGroup, Num=::NewEnumerator2)
     B=&MakeCustomFloatFromBool(Bool=$User.IsSkydiving, True Float=5, False Float=1)
    )
   ) {module=AddVelocity003}
   /SetVariables(Particles.Position=&MergeHelper_VectorSwitch(
     0=$Particles.ShapeLocation.ShapePosition
     1=$Particles.ShapeLocation001.ShapePosition
     2=$Particles.ShapeLocation002.ShapePosition
     3=$Particles.ShapeLocation002.ShapePosition
     Index=$Emitter.SpawnGroup
     Num=::NewEnumerator2
    )) {isSetParameters=true, module=SetVariables_#7}
   /SetVariables(Particles.SpawnGroup=$Emitter.SpawnGroup) {isSetParameters=true, module=SetVariables_#8}
  ]
  particleUpdateScript [
   &ParticleState()
   &ScaleSpriteSize(
    Scale Sprite Size Mode=::NewEnumerator0
    Uniform Scale Factor=&FloatFromCurve(FloatCurve=DataInterface() {propertyValues {Curve {keys [
         {arriveTangent=9.1187, interpMode=RCIM_Cubic, time=0, value=0}
         {arriveTangent=0.5833, interpMode=RCIM_Cubic, leaveTangent=0.5833, time=0.2487, value=1.1246}
         {arriveTangent=-1.497, interpMode=RCIM_Cubic, leaveTangent=-1.497, time=0.6, value=0.35}
         <at 1 0 RCIM_Cubic>
        ]}}})
   )
   &AccelerationForce(Acceleration=&Multiply_VectorByFloat(
     Vector=&RandomRangeVector(Minimum=<-3 -3 5>, Maximum=<3 3 10>, Evaluation Type=::NewEnumerator1)
     Float=1
    ))
   &Drag(Drag=5, Use Linear Drag=true, Use Rotational Drag=true)
   &GravityForce(Gravity=&MakeVector(
     X=0
     Y=0
     Z=&FloatFromCurve(FloatCurve=DataInterface() {propertyValues {Curve {keys [<at 0.3 0>,<at 1 1>]}}}, Scale Curve=400)
    ))
   &DynamicMaterialParameters(
    Index 0 Param 1=5
    Index 0 Param 2=12
    Index 0 Param 3=&FloatFromCurve(FloatCurve=DataInterface() {propertyValues {Curve {keys [<at 0 5 RCIM_Cubic>, {arriveTangent=-4.1151, interpMode=RCIM_Cubic, leaveTangent=-4.1151, tangentMode=RCTM_User, time=1, value=1}]}}})
    Module.Param0WriteEnabled=true
    Module.Param1WriteEnabled=true
    Module.Param2WriteEnabled=true
    Module.Param3WriteEnabled=false
   )
   &InheritSourceMovement(
    Applied Force Scale=&VectorFromFloat(Value=&FloatFromCurve(FloatCurve=DataInterface() {propertyValues {Curve {keys [<at 0 0>,<at 1 -0.5>]}}}))
    Applied Position Scale=&VectorFromFloat(Value=&MakeCustomFloatFromBool(
      Bool=$User.IsSkydiving
      True Float=&FloatFromCurve(FloatCurve=DataInterface() {propertyValues {Curve {keys [<at 0 0.95 RCIM_Cubic>,<at 1 0.8 RCIM_Cubic>]}}})
      False Float=&FloatFromCurve(FloatCurve=DataInterface() {propertyValues {Curve {keys [<at 0 0.95 RCIM_Cubic>,<at 1 0.8 RCIM_Cubic>]}}})
     ))
    ApplyForce=true
    ApplyForceFalloff=false
    ApplyPosition=true
    ApplyPositionFalloff=false
    Module.LimitForceSpeed=false
   ) {module=InheritSourceMovement001}
   &PointForce(Force Origin=$System.BodyKillPos, Force Strength=50, Module.Use Falloff=false, Module.UseLinearForce=false)
   &InheritSourceMovement(
    Applied Force Scale=<0 0 0>
    Applied Position Scale=&VectorFromFloat(Value=1)
    ApplyForce=true
    ApplyForceFalloff=false
    ApplyPosition=true
    ApplyPositionFalloff=false
    Module.LimitForceSpeed=false
   ) {enabled=false}
   /Niagara/Modules/Update/Forces/SpringForce(" Particle Equilibrium Position":NiagaraPosition=$Particles.Initial.Position, Dampening Coefficient=2, Spring Tightness=100) {enabled=false}
   &SolveForcesAndVelocity(Clamp Velocity=false, Limit Acceleration=false)
   &ScaleColor(
    Scale Alpha=&MakeCustomFloatFromBool(Bool=&Boolean_OrOperation(Bool1=$User.bisInVehicle, Bool2=$Emitter.ADS), True Float=0, False Float=1)
    ScaleA=true
    ScaleRGB=true
   ) {module=ScaleColor001}
   &ScaleColor(
    Scale RGB=&VectorFromFloat(Value=&FNTimeOfDayScalar(Night Scale=0.6, Morning Scale=1, Evening Scale=1))
    ScaleA=true
    ScaleRGB=true
   ) {module=ScaleColor002}
   &CameraOffset(Camera Offset Amount=&Multiply_Float(
     A=&FloatFromCurve(FloatCurve=DataInterface() {propertyValues {Curve {keys [{interpMode=RCIM_Cubic, tangentMode=RCTM_User, time=0.5, value=1}, {interpMode=RCIM_Cubic, tangentMode=RCTM_User, time=1, value=0}]}}})
     B=&MergeHelper_FloatSwitch(0=0, 1=5, 2=0, 3=0, 4=0, Index=$Particles.SpawnGroup, Num=::NewEnumerator3)
    ))
   /SetVariables(Particles.VisibilityTag=&MergeHelper_IntSwitch(
     0=&MakeCustomIntFromBool(Bool=$User.bisHeadHidden, True INT=1, False INT=0)
     1=0
     2=&MakeCustomIntFromBool(Bool=$User.bisHandHidden, True INT=1, False INT=0)
     3=&MakeCustomIntFromBool(Bool=$User.bisHandHidden, True INT=1, False INT=0)
     4=?
     5=?
     Index=$Particles.SpawnGroup
     Num=::NewEnumerator2
    )) {isSetParameters=true, module=SetVariables_#9}
   &KillParticlesInVolume(Sphere Radius=25, Volume Origin:Vector3f=$System.BodyKillPos)
  ]
  rendererClasses [&NiagaraSpriteRendererProperties]
  renderers [{rendererClass=&NiagaraSpriteRendererProperties}]
 }
 Goop_All001 {
  bEnabled=false
  emitterUpdateScript [
   &EmitterState()
   /SetVariables(Emitter.MovementSpawnAlpha:NiagaraFloat=&ClampFloat(Float=&Divide_Float(A=&VectorLength(Vector=$Engine.Owner.Velocity), B=500))) {isSetParameters=true, module=SetVariables_#4}
   &SpawnRate(
    Spawn Probability=0.75
    SpawnGroup:NiagaraInt32=&RandomRangeInt(Minimum=1, Maximum=3, Evaluation Type=::NewEnumerator1)
    SpawnRate=&Lerp_Float(A=8, B=3.5, Alpha=$Emitter.MovementSpawnAlpha)
    Use Spawn Probability=true
   )
   &SpawnRate(
    Spawn Probability=0.75
    SpawnRate=&Lerp_Float(A=8, B=4.5, Alpha=$Emitter.MovementSpawnAlpha)
    Use Spawn Probability=true
   ) {module=SpawnRate001}
   /SetVariables(Emitter.ADS:NiagaraBool=&SetBoolByFloatComparison(
     A=&DistanceBetweenPositions(
      A=&ReturnCameraPosition(
       Coordinate Space:ENiagaraCoordinateSpace=::Simulation
       Camera Query:NiagaraDataInterfaceCamera=DataInterface() {propertyValues {bRequireCurrentFrameData=false}}
      )
      B=$Engine.Emitter.SimulationPosition
     )
     B=&FN_FrontEndFloatModifier(Gameplay=250, Front End=0)
     Comparison Type=::NewEnumerator6
    )) {isSetParameters=true, module=SetVariables_#5}
  ]
  particleSpawnScript [
   &PartitionParticles(Number Of Partitions=4)
   /SetVariables(
    Particles.bIsArm:NiagaraBool=&SetBoolByIntComparison(A=$Emitter.SpawnGroup, B=2, Comparison Type=::NewEnumerator4)
    Particles.bIsEar:NiagaraBool=&SetBoolByIntComparison(A=$Emitter.SpawnGroup, B=0, Comparison Type=::NewEnumerator4)
    Particles.bIsHand:NiagaraBool=&SetBoolByIntComparison(A=$Emitter.SpawnGroup, B=3, Comparison Type=::NewEnumerator4)
    Particles.bIsMainChest:NiagaraBool=&SetBoolByIntComparison(A=$Emitter.SpawnGroup, B=1, Comparison Type=::NewEnumerator4)
   ) {isSetParameters=true, module=SetVariables_#6}
   &InitializeParticle(
    Color=<rgba 0.8 0.8 0.8 3>
    Lifetime=&Multiply_Float(
     A=&Lerp_Float(
      A=&RandomRangeFloat(Minimum=1.4, Maximum=2.2)
      B=0.8
      Alpha=&Divide_Float(A=&VectorLength(Vector=$Engine.Owner.Velocity), B=5000)
     )
     B=&MergeHelper_FloatSwitch(0=0.8, 1=0.7, 2=1, 3=1, Index=$Emitter.SpawnGroup, Num=::NewEnumerator2)
    )
    Mesh Scale Mode=::NewEnumerator1
    Mesh Uniform Scale Max=0.4
    Mesh Uniform Scale Min=0.25
    Sprite Size Mode=::NewEnumerator1
    UsePositionOffset=false
    Write Material Random=true
   )
   &InitialMeshOrientation(Use Rotation Vector=false)
   &SkeletalMeshLocation(
    Bone / Socket Index=&MergeHelper_IntSwitch(
     0=&RandomRangeInt(Minimum=0, Maximum=1)
     1=2
     2=&RandomRangeInt(Minimum=3, Maximum=4)
     3=&RandomRangeInt(Minimum=5, Maximum=6)
     4=?
     5=?
     Index=$Emitter.SpawnGroup
     Num=::NewEnumerator2
    )
    MaskSpawnGroup=false
    Mesh Sampling Type=::NewEnumerator5
    Orientation Sampling=::NewEnumerator1
    Skeletal Mesh=$System.SkeletalMesh
    Socket Sampling Mode=::NewEnumerator5
   )
   &ShapeLocation(
    Additional Quaternion Rotation=true
    Additional Yaw / Pitch / Roll=<0 90 0>
    Apply To Particle Position=::NewEnumerator1
    Cylinder Height=20
    Cylinder Radius=3
    MaskSpawnGroup=false
    Override Seed=false
    Rotation Angle Type:ENiagara_AngleInput=::NewEnumerator0
    Rotation Mode=::NewEnumerator3
    Rotation Quaternion=$Particles.SkeletalMeshLocation.SampledRotation
    Shape Primitive=::NewEnumerator1
    Surface Only=false
    UseOwnerScale=true
   )
   &ShapeLocation(
    Additional Quaternion Rotation=true
    Additional Yaw / Pitch / Roll=<0 0 0.25>
    Apply To Particle Position=::NewEnumerator1
    Handle Radius=4
    Large Radius=7
    MaskSpawnGroup=false
    Override Seed=false
    Rotation Mode=::NewEnumerator3
    Rotation Quaternion=$Particles.SkeletalMeshLocation.SampledRotation
    Shape Primitive=::NewEnumerator3
    UseOwnerScale=true
   ) {module=ShapeLocation001}
   &ShapeLocation(
    Apply To Particle Position=::NewEnumerator1
    MaskSpawnGroup=false
    Sphere Distribution=::NewEnumerator1
    Sphere Radius=&MergeHelper_FloatSwitch(0=0, 1=0, 2=11.5, 3=5, 4=0, Index=$Emitter.SpawnGroup, Num=::NewEnumerator3)
    U Position=&RandomRangeFloat(Minimum=0, Maximum=1)
    UseOwnerScale=true
    V Position=&RandomRangeFloat(Minimum=0, Maximum=0.5)
   ) {module=ShapeLocation002}
   &AddVelocity(
    Velocity=&MergeHelper_VectorSwitch(
     0=$Particles.ShapeLocation.ShapeNormal
     1=$Particles.ShapeLocation001.ShapeNormal
     2=$Particles.ShapeLocation002.ShapeNormal
     3=$Particles.ShapeLocation002.ShapeNormal
     Index=$Emitter.SpawnGroup
     Num=::NewEnumerator2
    )
    Velocity Speed Scale=&Multiply_Float(
     A=&MergeHelper_FloatSwitch(0=20, 1=0, 2=40, 3=90, Index=$Emitter.SpawnGroup, Num=::NewEnumerator2)
     B=&MakeCustomFloatFromBool(Bool=$User.IsSkydiving, True Float=5, False Float=1)
    )
   ) {module=AddVelocity003}
   /SetVariables(Particles.Position=&MergeHelper_VectorSwitch(
     0=$Particles.ShapeLocation.ShapePosition
     1=$Particles.ShapeLocation001.ShapePosition
     2=$Particles.ShapeLocation002.ShapePosition
     3=$Particles.ShapeLocation002.ShapePosition
     Index=$Emitter.SpawnGroup
     Num=::NewEnumerator2
    )) {isSetParameters=true, module=SetVariables_#7}
   /SetVariables(Particles.SpawnGroup=$Emitter.SpawnGroup) {isSetParameters=true, module=SetVariables_#8}
  ]
  particleUpdateScript [
   &ParticleState()
   &ScaleSpriteSize(
    Scale Sprite Size Mode=::NewEnumerator0
    Uniform Scale Factor=&FloatFromCurve(FloatCurve=DataInterface() {propertyValues {Curve {keys [
         {arriveTangent=9.1187, interpMode=RCIM_Cubic, time=0, value=0}
         {arriveTangent=0.5833, interpMode=RCIM_Cubic, leaveTangent=0.5833, time=0.2487, value=1.1246}
         {arriveTangent=-1.497, interpMode=RCIM_Cubic, leaveTangent=-1.497, time=0.6, value=0.35}
         <at 1 0 RCIM_Cubic>
        ]}}})
   )
   &AccelerationForce(Acceleration=&Multiply_VectorByFloat(
     Vector=&RandomRangeVector(Minimum=<-3 -3 5>, Maximum=<3 3 10>, Evaluation Type=::NewEnumerator1)
     Float=1
    ))
   &Drag(Drag=5, Use Linear Drag=true, Use Rotational Drag=true)
   &GravityForce(Gravity=&MakeVector(
     X=0
     Y=0
     Z=&FloatFromCurve(FloatCurve=DataInterface() {propertyValues {Curve {keys [<at 0.3 0>,<at 1 1>]}}}, Scale Curve=400)
    ))
   &DynamicMaterialParameters(
    Index 0 Param 1=5
    Index 0 Param 2=12
    Index 0 Param 3=&FloatFromCurve(FloatCurve=DataInterface() {propertyValues {Curve {keys [<at 0 5 RCIM_Cubic>, {arriveTangent=-4.1151, interpMode=RCIM_Cubic, leaveTangent=-4.1151, tangentMode=RCTM_User, time=1, value=1}]}}})
    Module.Param0WriteEnabled=true
    Module.Param1WriteEnabled=true
    Module.Param2WriteEnabled=true
    Module.Param3WriteEnabled=false
   )
   &InheritSourceMovement(
    Applied Force Scale=&VectorFromFloat(Value=&FloatFromCurve(FloatCurve=DataInterface() {propertyValues {Curve {keys [<at 0 0>,<at 1 -0.5>]}}}))
    Applied Position Scale=&VectorFromFloat(Value=&MakeCustomFloatFromBool(
      Bool=$User.IsSkydiving
      True Float=&FloatFromCurve(FloatCurve=DataInterface() {propertyValues {Curve {keys [<at 0 0.95 RCIM_Cubic>,<at 1 0.8 RCIM_Cubic>]}}})
      False Float=&FloatFromCurve(FloatCurve=DataInterface() {propertyValues {Curve {keys [<at 0 0.95 RCIM_Cubic>,<at 1 0.8 RCIM_Cubic>]}}})
     ))
    ApplyForce=true
    ApplyForceFalloff=false
    ApplyPosition=true
    ApplyPositionFalloff=false
    Module.LimitForceSpeed=false
   ) {enabled=false, module=InheritSourceMovement001}
   &PointForce(Force Origin=$System.BodyKillPos, Force Strength=50, Module.Use Falloff=false, Module.UseLinearForce=false)
   &InheritSourceMovement(ApplyForce=true, ApplyForceFalloff=false, ApplyPosition=true, ApplyPositionFalloff=false, Module.LimitForceSpeed=false) {enabled=false}
   &SolveForcesAndVelocity(Clamp Velocity=false, Limit Acceleration=false)
   &ScaleColor(
    Scale Alpha=&MakeCustomFloatFromBool(Bool=&Boolean_OrOperation(Bool1=$User.bisInVehicle, Bool2=$Emitter.ADS), True Float=0, False Float=1)
    ScaleA=true
    ScaleRGB=true
   ) {module=ScaleColor001}
   &ScaleColor(
    Scale RGB=&VectorFromFloat(Value=&FNTimeOfDayScalar(Night Scale=0.6, Morning Scale=1, Evening Scale=1))
    ScaleA=true
    ScaleRGB=true
   ) {module=ScaleColor002}
   &CameraOffset(Camera Offset Amount=&Multiply_Float(
     A=&FloatFromCurve(FloatCurve=DataInterface() {propertyValues {Curve {keys [{interpMode=RCIM_Cubic, tangentMode=RCTM_User, time=0.5, value=1}, {interpMode=RCIM_Cubic, tangentMode=RCTM_User, time=1, value=0}]}}})
     B=&MergeHelper_FloatSwitch(0=0, 1=5, 2=0, 3=0, 4=0, Index=$Particles.SpawnGroup, Num=::NewEnumerator3)
    ))
   /SetVariables(Particles.VisibilityTag=&MergeHelper_IntSwitch(
     0=0
     1=0
     2=0
     3=&MakeCustomIntFromBool(Bool=$User.bisHandHidden, True INT=1, False INT=0)
     4=?
     5=?
     Index=$Particles.SpawnGroup
     Num=::NewEnumerator2
    )) {isSetParameters=true, module=SetVariables_#9}
   &KillParticlesInVolume(Sphere Radius=25, Volume Origin:Vector3f=$System.BodyKillPos)
  ]
  rendererClasses [&NiagaraSpriteRendererProperties]
  renderers [{rendererClass=&NiagaraSpriteRendererProperties}]
 }
}
summary {
 systemName=NS_FuzzyDarkness_Char_Idle
 emitters [
  {emitterName=Goop_All, rendererClasses [&NiagaraSpriteRendererProperties]}
  {bEnabled=false, emitterName=Goop_All001, rendererClasses [&NiagaraSpriteRendererProperties]}
 ]
 userVariables [
  User.bIsFrontEnd {type=NiagaraBool}
  User.bIsFrontEndPreview {type=NiagaraBool}
  User.bisHandHidden {type=NiagaraBool}
  User.bisHeadHidden {type=NiagaraBool}
  User.bisInVehicle {type=NiagaraBool}
  User.IsSkydiving {type=NiagaraBool}
 ]
}
