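# Material function for the EctoCat character that flipbooks between eye-pose mesh pieces using World Position Offset.
# Pose parameters cross thresholds in a chained If priority (blink, brow down/up, glabella down/up, squint, default) to choose which vertex-color value VertexColorMask matches.
# Unselected pieces get a huge negative offset to hide them, while channel values above 0.95 are exempt.
# The UseEyesFlipbookWPO static switch gates adding this to the incoming material attributes.
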MaterialFunction /Game/Characters/Player/Male/Large/Bodies/M_LRG_EctoCat/Materials/MaterialFunctions/MF_FlipbookEyesWPO
graph {nodes [
  n1 = GetMaterialAttributes(None=FunctionInput() {inputName=Material, inputType=FunctionInput_MaterialAttributes}) {attributeGetTypes [#1]}
  n2 = ChannelMaskParameter(None=VertexColor()) {group=Eye_Flipbook_Settings, parameterName=VertexColorChannel, sortPriority=2}
  n3 = MaterialFunctionCall!(
   Hide Element Toggle=Constant() {r=1}
   VertexColorChannel=@n2.R
   Vertex Value to hide=ScalarParameter() {defaultValue=0.4, group=Eye_Flipbook_Settings, parameterName="DefaultPose_VertexValue(R)", sortPriority=14}
  ) {materialFunction=/Game/Packages/Fortress_SharedMaterials/Base_Material_Functions/VertexColorMask}
  n4 = If(
   A=ScalarParameter() {group=Eye_Poses, parameterName=R_squint_inner_pose, sortPriority=16}
   B=ScalarParameter() {defaultValue=0.375, group=Eye_Flipbook_Settings, parameterName=Threshold_squint_inner_pose}
   "A > B"=MaterialFunctionCall(
    Hide Element Toggle=Constant() {r=1}
    VertexColorChannel=@n2.R
    Vertex Value to hide=ScalarParameter() {defaultValue=0.2, group=Eye_Flipbook_Settings, parameterName="SquintPose_VertexValue(R)", sortPriority=19}
   ) {materialFunction=&VertexColorMask}.Result
   "A == B"=@n3.Result
   "A < B"=@n3.Result
  )
  n5 = If(
   A=ScalarParameter() {group=Eye_Poses, parameterName=C_glabella_up_pose, sortPriority=19}
   B=ScalarParameter() {defaultValue=0.1, group=Eye_Flipbook_Settings, parameterName=Threshold_Glabella_up}
   "A > B"=MaterialFunctionCall(
    Hide Element Toggle=Constant() {r=1}
    VertexColorChannel=@n2.R
    Vertex Value to hide=ScalarParameter() {defaultValue=0.7, group=Eye_Flipbook_Settings, parameterName="GlabellaUp_VertexValue(R)", sortPriority=17}
   ) {materialFunction=&VertexColorMask}.Result
   "A == B"=@n4
   "A < B"=@n4
  )
  n6 = If(
   A=ScalarParameter() {group=Eye_Poses, parameterName=C_glabella_down_pose, sortPriority=18}
   B=ScalarParameter() {defaultValue=0.1, group=Eye_Flipbook_Settings, parameterName=Threshold_Glabella_down}
   "A > B"=MaterialFunctionCall(
    Hide Element Toggle=Constant() {r=1}
    VertexColorChannel=@n2.R
    Vertex Value to hide=ScalarParameter() {defaultValue=0.6, group=Eye_Flipbook_Settings, parameterName="GlabellaDown_VertexValue(R)", sortPriority=18}
   ) {materialFunction=&VertexColorMask}.Result
   "A == B"=@n5
   "A < B"=@n5
  )
  n7 = If(
   A=ScalarParameter() {group=Eye_Poses, parameterName=R_brow_up_pose, sortPriority=17}
   B=ScalarParameter() {defaultValue=0.1, group=Eye_Flipbook_Settings, parameterName=Threshold_R_brow_up_pose}
   "A > B"=MaterialFunctionCall(
    Hide Element Toggle=Constant() {r=1}
    VertexColorChannel=@n2.R
    Vertex Value to hide=ScalarParameter() {defaultValue=0.5, group=Eye_Flipbook_Settings, parameterName="BrowUp_VertexValue(R)", sortPriority=15}
   ) {materialFunction=&VertexColorMask}.Result
   "A == B"=@n6
   "A < B"=@n6
  )
  n8 = If(
   A=ScalarParameter() {group=Eye_Poses, parameterName=R_brow_down_pose, sortPriority=15}
   B=ScalarParameter() {defaultValue=0.1, group=Eye_Flipbook_Settings, parameterName=Threshold_R_brow_down_pose}
   "A > B"=MaterialFunctionCall(
    Hide Element Toggle=Constant() {r=1}
    VertexColorChannel=@n2.R
    Vertex Value to hide=ScalarParameter() {defaultValue=0.3, group=Eye_Flipbook_Settings, parameterName="BrowDown_VertexValue(R)", sortPriority=16}
   ) {materialFunction=&VertexColorMask}.Result
   "A == B"=@n7
   "A < B"=@n7
  )
  n9 = If(
   A=ScalarParameter() {group=Eye_Poses, parameterName=R_blink_pose, sortPriority=20}
   B=ScalarParameter() {defaultValue=0.1, group=Eye_Flipbook_Settings, parameterName=Threshold_R_blink_pose}
   "A > B"=MaterialFunctionCall(
    Hide Element Toggle=Constant() {r=1}
    VertexColorChannel=@n2.R
    Vertex Value to hide=ScalarParameter() {defaultValue=0.1, group=Eye_Flipbook_Settings, parameterName="BlinkPose_VertexValue(R)", sortPriority=20}
   ) {materialFunction=&VertexColorMask}.Result
   "A == B"=@n8
   "A < B"=@n8
  )
  n10 = OneMinus(None=If(A=@n2.R, B=Constant() {r=0.95}, "A > B"=Constant() {r=1}, "A == B"=@n9, "A < B"=@n9))
  n11 = Add(A=@n1.WorldPositionOffset, B=Multiply(A=Constant() {r=-99999997952}, B=Saturate(None=@n10)))
  FunctionOutput(None=SetMaterialAttributes!(
    MaterialAttributes=@n1.MaterialAttributes
    WorldPositionOffset=StaticSwitchParameter(True=@n11, False=@n1.WorldPositionOffset) {group=Eye_Flipbook_Settings, parameterName=UseEyesFlipbookWPO, sortPriority=0}
   ) {attributeSetTypes [#1]})
  Multiply(A=Subtract(A=WorldPosition().XYZ, B=ObjectPositionWS())) {constB=-1}
 ]}
