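# Parent material for the eyes of the FuzzyDarkness player cosmetic body, unlit and translucent on skeletal meshes.
# Emissive is a flat bright white scaled by a time-of-day function that dims it at night, and the FuzzyFlipbook texture's red channel drives opacity.
# The attributes pass through EctoCat's MF_FlipbookEyesWPO function, then a small vertex-normal push is added to its World Position Offset.
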Material /BRCosmetics/Characters/Player/Male/Medium/Bodies/M_LRG_FuzzyDarkness/Materials/M_FuzzyDarkness_Eyes_PARENT
graph {
 properties {
  bUseMaterialAttributes=true
  bUsedWithMeshDeformer=true
  bUsedWithSkeletalMesh=true
  blendMode=BLEND_Translucent
  pixelDepthOffsetMode=PDOM_Legacy
  shadingModel=MSM_Unlit
  usedShadingModels=Unlit
 }
 nodes [
  Multiply(B=Constant3Vector() {constant=<rgba 1 0 0 0>})
  PreSkinnedNormal()
  n1 = Constant3Vector()
  n2 = Constant() {r=1}
  n3 = Multiply(
   A=MaterialFunctionCall(Value_Morning=@n2, Value_Day=@n2, Value_Evening=@n2, Value_Night=Constant() {r=0.5}) {materialFunction=&Scalar_Time_of_Day}.Result
   B=Constant3Vector() {constant=<rgba 1.4 1.4 1.4 0>}
  )
  n4 = TextureSampleParameter2D() {
   group=Flipbook_Lines
   parameterName=FuzzyFlipbook
   texture=/BRCosmetics/Characters/Player/Male/Medium/Bodies/M_LRG_FuzzyDarkness/Textures/T_FuzzyDarkness_FaceAcc_Extra_FX
  }
  n5 = GetMaterialAttributes(None=MaterialFunctionCall!(Material=SetMaterialAttributes!(MaterialAttributes, Base Color=@n1.R, Emissive Color=@n3, Opacity=@n4.R) {attributeSetTypes [#1, #2, #3]}) {materialFunction=/Game/Characters/Player/Male/Large/Bodies/M_LRG_EctoCat/Materials/MaterialFunctions/MF_FlipbookEyesWPO}.Result) {attributeGetTypes [#4]}
  n6 = VertexNormalWS()
  SetMaterialAttributes!(MaterialAttributes=@n5.MaterialAttributes, World Position Offset=Add(A=@n5.World Position Offset, B=Multiply(B=@n6) {constA=0.1})) {attributeSetTypes [#4]}
 ]
}
