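# Parent character material for the FuzzyDarkness body, built on the 2023 CharacterParent PreFX and PostFX functions, with animated emissive shading.
# Triplanar-blended panning cloud noise on pre-skinned position distorts scrolling 3D gradient noise, which modulates emissive alongside a slow sine pulse.
# Time-of-day scalars adjust emissive brightness and night saturation.
# EctoCat's flipbook-eyes WPO function follows, then base color above a threshold turns black with white, time-of-day-scaled emissive.
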Material /BRCosmetics/Characters/Player/Male/Medium/Bodies/M_LRG_FuzzyDarkness/Materials/M_FuzzyDarkness_PARENT
graph {
 properties {
  bUseMaterialAttributes=true
  bUsedWithMeshDeformer=true
  bUsedWithMorphTargets=true
  bUsedWithSkeletalMesh=true
  pixelDepthOffsetMode=PDOM_Legacy
  shadingModel=MSM_FromMaterialExpression
  usedShadingModels=Default Lit | Subsurface | Clear Coat | Subsurface Profile | Cloth
 }
 nodes [
  Constant() {r=0.87}
  n1 = MaterialFunctionCall() {materialFunction=&MF_CharacterParent_2023_PreFX}
  n2 = GetMaterialAttributes(None=MaterialFunctionCall(Material=@n1.Result, M=@n1.Result, Pre Skinned Local Position=@n1.Result, Distance Blend=@n1.Result, Fresnel=@n1.Result) {materialFunction=&MF_CharacterParent_2023_PostFX}.Result) {attributeGetTypes [#1]}
  n3 = VertexInterpolator(VS=PreSkinnedPosition())
  n4 = Multiply(A=Time(), B=Constant3Vector() {constant=<rgba 0 0 -1.35 0>})
  n5 = VertexInterpolator(VS=PreSkinnedPosition())
  n6 = Time()
  n7 = Constant() {desc=Speed, r=-0.75}
  n8 = Multiply(A=@n6, B=MaterialFunctionCall(B=@n7) {materialFunction=&AppendMany}.RGB)
  n9 = Constant() {desc=UV Scale, r=0.04}
  n10 = ComponentMask(None=Multiply(A=Add(A=@n5.PS, B=@n8), B=@n9) {constB=0.1}) {b=true, g=true}
  n11 = Multiply(A=@n6, B=MaterialFunctionCall(R=@n7, B=@n7) {materialFunction=&AppendMany}.RGB)
  n12 = ComponentMask(None=Multiply(A=Add(A=@n5.PS, B=@n11), B=@n9) {constB=0.1}) {b=true, r=true}
  n13 = Transform(None=VertexNormalWS()) {transformSourceType=TRANSFORMSOURCE_World, transformType=TRANSFORM_Local}
  n14 = Multiply(A=@n13) {constB=-1}
  n15 = Add(A=Saturate(None=ComponentMask(None=@n13) {g=true}), B=Saturate(None=ComponentMask(None=@n14) {g=true}))
  n16 = MaterialFunctionCall(R=@n7) {materialFunction=&AppendMany}
  n17 = Multiply(A=Add(A=@n5.PS, B=Multiply(A=@n6, B=@n16.RGB)), B=@n9)
  n18 = Saturate(None=ComponentMask(None=@n13) {b=true})
  n19 = Saturate(None=ComponentMask(None=Multiply(A=@n13) {constB=-1}) {b=true})
  n20 = LinearInterpolate(
   A=LinearInterpolate(
    A=TextureSample(UVs=@n10) {texture=&AtmosphericCloudNoise01}.R
    B=TextureSample(UVs=@n12) {texture=&AtmosphericCloudNoise01}.R
    Alpha=SmoothStep(Value=@n15) {constMin=0.4}
   )
   B=TextureSample(UVs=ComponentMask(None=@n17) {g=true, r=true}) {texture=&AtmosphericCloudNoise01}.R
   Alpha=SmoothStep(Value=Add(A=@n18, B=@n19)) {constMin=0.4}
  )
  n21 = Noise!(World Position=Add(A=Add(A=@n3.PS, B=@n4), B=Multiply(B=@n20) {constA=2}), FilterWidth) {levels=3, noiseFunction=NOISEFUNCTION_GradientTex3D, outputMax=1.5, outputMin=0.55, scale=0.12}
  n22 = Constant() {r=0.02}
  n23 = MaterialFunctionCall(Value_Morning=@n22, Value_Day=@n22, Value_Evening=@n22, Value_Night=Constant() {r=0.11}) {materialFunction=&Scalar_Time_of_Day}
  n24 = Multiply(A=@n2.Emissive Color, B=Saturate(None=Add(A=@n21, B=@n23.Result) {constB=0.6}))
  n25 = Sine(None=Multiply(A=Time()) {constB=0.3})
  n26 = LinearInterpolate(Alpha=Add(A=Multiply(A=@n25) {constB=0.5}) {constB=0.5}) {constA=1.1, constB=1.3}
  n27 = Constant() {r=1}
  n28 = Constant() {r=0.87}
  n29 = Constant() {r=0.25}
  n30 = Constant()
  n31 = Desaturation(
   None=Multiply(
    A=Multiply(A=@n24, B=@n26)
    B=MaterialFunctionCall(Value_Morning=@n27, Value_Day=@n27, Value_Evening=@n28, Value_Night=@n29) {materialFunction=&Scalar_Time_of_Day}.Result
   )
   Fraction=MaterialFunctionCall(Value_Morning=@n30, Value_Day=@n30, Value_Evening=@n30, Value_Night=Constant() {r=-0.7}) {materialFunction=&Scalar_Time_of_Day}.Result
  )
  n32 = GetMaterialAttributes(None=MaterialFunctionCall!(Material=SetMaterialAttributes(MaterialAttributes=@n2.MaterialAttributes, Emissive Color=@n31) {attributeSetTypes [#1]}) {materialFunction=/Game/Characters/Player/Male/Large/Bodies/M_LRG_EctoCat/Materials/MaterialFunctions/MF_FlipbookEyesWPO}.Result) {attributeGetTypes [#2, #1]}
  n33 = Constant3Vector() {constant=<rgba 1 1 1 0>}
  n34 = Constant() {r=1}
  n35 = MaterialFunctionCall(Value_Morning=@n34, Value_Day=@n34, Value_Evening=@n34, Value_Night=Constant() {r=0.4}) {materialFunction=&Scalar_Time_of_Day}
  n36 = Constant() {r=1}
  n37 = If(A=@n32.Base Color, "A > B"=@n36, "A == B"=@n36, "A < B"=Constant()) {constB=0.7}
  SetMaterialAttributes!(
   MaterialAttributes=@n32.MaterialAttributes
   Emissive Color=LinearInterpolate(A=@n32.Emissive Color, B=Multiply(A=@n33, B=@n35.Result), Alpha=@n37)
   Base Color=LinearInterpolate(A=@n32.Base Color, Alpha=@n37) {constB=0}
  ) {attributeSetTypes [#1, #2]}
 ]
}
