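# Physics asset for the FuzzyDarkness male medium character body that provides collision and joint setup for its dynamic tail.
# A pelvis sphere anchors a tiny box at the tail base, followed by three progressively smaller spheres along the tail bones.
# The base-to-pelvis joint rotates freely on all angular axes, while the downstream chain links use tight swing limits and even tighter twist limits.
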PhysicsAsset /BRCosmetics/Characters/Player/Male/Medium/Bodies/M_LRG_FuzzyDarkness/Meshes/M_LRG_FuzzyDarkness_Physics
bodies {
 dyn_tail_01 [{shapeType=Sphere, radius=8.4013, center=<5.2688 -0.3938 -0.3286>}]
 dyn_tail_01_Base [{shapeType=Box, extentX=1.01, extentY=1.01, extentZ=1.01, rotation=<rot 0 -90 0>}]
 dyn_tail_02 [{shapeType=Sphere, radius=6.8643, center=<1.9397 -0.6517 0.0532>}]
 dyn_tail_03 [{shapeType=Sphere, radius=4.6829, center=<1.0741 0.7821 -0.4969>}]
 pelvis [{shapeType=Sphere, radius=6.3961, center=<-0.2274 -0.6784 -0.1218>}]
}
constraints [
 {
  bone1Name=dyn_tail_01_Base
  bone2Name=pelvis
  swing1LimitDegrees=45
  swing1Motion=Free
  swing2LimitDegrees=45
  swing2Motion=Free
  twistLimitDegrees=45
  twistMotion=Free
 }
 {bone1Name=dyn_tail_01, bone2Name=dyn_tail_01_Base, swing1LimitDegrees=18, swing2LimitDegrees=18, twistLimitDegrees=6}
 {bone1Name=dyn_tail_02, bone2Name=dyn_tail_01, swing1LimitDegrees=18, swing2LimitDegrees=18, twistLimitDegrees=6}
 {bone1Name=dyn_tail_03, bone2Name=dyn_tail_02, swing1LimitDegrees=18, swing2LimitDegrees=18, twistLimitDegrees=6}
]
