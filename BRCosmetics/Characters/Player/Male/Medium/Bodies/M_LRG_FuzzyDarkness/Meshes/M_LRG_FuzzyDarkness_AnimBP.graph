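# Post-process style AnimBP for the FuzzyDarkness player body cosmetic, layering secondary motion onto a pose copied from the attached parent mesh.
# RigidBody physics and Control Rig run between additive bone offsets for shoulder belts, tail and cape, some gated by front-end, parachute and skydive states.
# Each update derives flags from bRigidBodyEnabled, applying cape/tail offsets without physics and collapsing the tail base to zero scale when also in the front end.
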AnimBlueprint /BRCosmetics/Characters/Player/Male/Medium/Bodies/M_LRG_FuzzyDarkness/Meshes/M_LRG_FuzzyDarkness_AnimBP
target_skeleton=&/M_LRG_FuzzyDarkness_Skeleton
graphs [
 AnimGraph {nodes [
   n1 = NOT Boolean(A=Get bIsInFrontEnd())
   n2 = "Transform (Modify) Bone - Bone: dyn_shoulderbelt_l"!(
    ComponentPose=Local To Component(LocalPose=Copy Pose From Mesh() {bUseAttachedParent=true})
    bAlphaBoolEnabled:bool=@n1
    Translation="3.700,1.533,5.448"
    Rotation="10.034,-1.890,2.123"
    Scale="1,1,1"
   ) -> (Pose) {
    alphaInputType=Bool
    rotationMode=BMM_Additive
    rotationSpace=BCS_BoneSpace
    translationMode=BMM_Additive
    translationSpace=BCS_BoneSpace
    alphaBoolBlend {blendOption=Linear}
    boneToModify {boneName=dyn_shoulderbelt_l}
    rotation=<rot 10.034 -1.8899 2.1229>
    translation=<3.7 1.533 5.448>
   }
   n3 = "Transform (Modify) Bone - Bone: dyn_shoulderbelt_r"!(ComponentPose=@n2, bAlphaBoolEnabled:bool=@n1, Translation="1.389,-0.690,-1.723", Rotation="17.284,7.381,-2.232", Scale="1,1,1") -> (Pose) {
    alphaInputType=Bool
    rotationMode=BMM_Additive
    rotationSpace=BCS_BoneSpace
    translationMode=BMM_Additive
    translationSpace=BCS_BoneSpace
    alphaBoolBlend {blendOption=Linear}
    boneToModify {boneName=dyn_shoulderbelt_r}
    rotation=<rot 17.284 7.381 -2.232>
    translation=<1.389 -0.69 -1.723>
   }
   n4 = RigidBody(ComponentPose=@n3) {
    bClampLinearTranslationLimitToRefPose=true
    bOverrideWorldGravity=true
    overridePhysicsAsset=&/M_LRG_FuzzyDarkness_Physics
    simulationSpace=BaseBoneSpace
    alphaBoolBlend {blendOption=Linear}
    baseBoneRef {boneName=root}
    componentAppliedLinearAccClamp=<3000 3000 3000>
    componentLinearAccScale=<0.45 0.45 0.9>
    componentLinearVelScale=<0.25 0.25 0.3>
    overrideWorldGravity {z=-980}
   }
   n5 = "Transform (Modify) Bone - Bone: dyn_tail_01_Base"!(ComponentPose=@n4, bAlphaBoolEnabled:bool=Get bIsParachuteOpened(), Translation="0,0,0", Rotation="0,-40,0", Scale="1,1,1") -> (Pose) {
    alphaInputType=Bool
    rotationMode=BMM_Additive
    rotationSpace=BCS_BoneSpace
    translationMode=BMM_Additive
    translationSpace=BCS_BoneSpace
    alphaBoolBlend {blendInTime=0.2, blendOption=Linear, blendOutTime=0.2}
    boneToModify {boneName=dyn_tail_01_Base}
    rotation {yaw=-40}
   }
   n6 = "Transform (Modify) Bone - Bone: attach_cape"!(ComponentPose=@n5, bAlphaBoolEnabled:bool=Get bIsSkydiveFloating(), Translation="0,-3,-1", Rotation="0,0,0", Scale="1,1,1") -> (Pose) {
    alphaInputType=Bool
    rotationMode=BMM_Additive
    rotationSpace=BCS_BoneSpace
    translationMode=BMM_Additive
    translationSpace=BCS_BoneSpace
    alphaBoolBlend {blendInTime=0.2, blendOption=Linear, blendOutTime=0.2}
    boneToModify {boneName=attach_cape}
    translation {y=-3, z=-1}
   }
   n7 = Control Rig(Source=Component To Local(ComponentPose=@n6)) {
    alphaBoolBlend {blendOption=Linear}
    controlRigAssetReference {blueprintRigClass=&/M_LRG_FuzzyDarkness_CtrlRig.M_LRG_FuzzyDarkness_CtrlRig_C}
   }
   n8 = "Transform (Modify) Bone - Bone: dyn_tail_01"(ComponentPose=Local To Component(LocalPose=@n7), Alpha=1, Translation="0,0,0", Rotation="0,-20,0", Scale="1,1,1") {
    rotationMode=BMM_Additive
    rotationSpace=BCS_BoneSpace
    alphaBoolBlend {blendOption=Linear}
    boneToModify {boneName=dyn_tail_01}
    rotation {yaw=-20}
   }
   n9 = "Transform (Modify) Bone - Bone: dyn_tail_02"(ComponentPose=@n8, Alpha=1, Translation="0,0,0", Rotation="0,-15,0", Scale="1,1,1") {
    rotationMode=BMM_Additive
    rotationSpace=BCS_BoneSpace
    alphaBoolBlend {blendOption=Linear}
    boneToModify {boneName=dyn_tail_02}
    rotation {yaw=-15}
   }
   n10 = Get is Rigid body not() -> (is Rigid body not:double)
   n11 = "Transform (Modify) Bone - Bone: attach_cape"!(ComponentPose=@n9, Alpha:float=@n10, Translation="0,0,0", Rotation="0,0,4", Scale="1,1,1") -> (Pose) {
    rotationMode=BMM_Additive
    rotationSpace=BCS_BoneSpace
    alphaBoolBlend {blendOption=Linear}
    boneToModify {boneName=attach_cape}
    rotation {roll=4}
   }
   n12 = "Transform (Modify) Bone - Bone: dyn_tail_01_Base"!(ComponentPose=@n11, Alpha:float=@n10, Translation="0,0,0", Rotation="0,30,0", Scale="1,1,1") -> (Pose) {
    rotationMode=BMM_Additive
    rotationSpace=BCS_BoneSpace
    scaleMode=BMM_Replace
    scaleSpace=BCS_BoneSpace
    translationMode=BMM_Additive
    translationSpace=BCS_BoneSpace
    alphaBoolBlend {blendOption=Linear}
    boneToModify {boneName=dyn_tail_01_Base}
    rotation {yaw=30}
   }
   n13 = "Transform (Modify) Bone - Bone: dyn_tail_01_Base"!(
    ComponentPose=@n12
    bAlphaBoolEnabled:bool=Get FE Rigid body disabled() -> (FE Rigid body disabled:bool)
    Translation="-15.000,20.000,0.000"
    Rotation="0,0,0"
    Scale="0,0,0"
   ) -> (Pose) {
    alphaInputType=Bool
    rotationMode=BMM_Additive
    rotationSpace=BCS_BoneSpace
    scaleMode=BMM_Replace
    scaleSpace=BCS_BoneSpace
    translationMode=BMM_Additive
    translationSpace=BCS_BoneSpace
    alphaBoolBlend {blendOption=Linear}
    boneToModify {boneName=dyn_tail_01_Base}
    scale=<0 0 0>
    translation=<-15 20>
   }
   Output Pose(Result=Component To Local(ComponentPose=@n13)) {layerGroup=None}
  ]}
 EventGraph {nodes [
   Event BlueprintUpdateAnimation()
   > Set Skeletal Mesh(Skeletal Mesh:Skeletal Mesh Component Object Reference=GetOwnerSkeletalMeshComponent()) -> (Output_Get:Skeletal Mesh Component Object Reference)
   n1 = Get bRigidBodyEnabled()
   > Set is Rigid body not(is Rigid body not:double=SelectFloat(A=0, B=1, bPickA=@n1)) -> (Output_Get:double)
   > Set FE Rigid body disabled(FE Rigid body disabled:bool=AND Boolean(A=NOT Boolean(A=@n1), B=Get bIsInFrontEnd())) -> (Output_Get:bool)
  ]}
]
variables [
 FE Rigid body disabled {type=boolean}
 is Rigid body not {type=number}
 Skeletal Mesh {type=SkeletalMeshComponent}
]
